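; Dynamic UI director for Cherry Smoke boons that tracks the Hot Hands and Action Reload gameplay tags on the current view target.
; A view target change passes the new target to one wait function per tag.
; Each checks the actor and owning local controller, casts the controlled pawn to FortPlayerPawnAthena, then starts tag add and remove wait tasks.
; No responses to those tasks appear in the graph, and BeginPlay is empty.
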(event EventBeginPlay)

(event Custom|OnViewTargetChanged
  (params (PlayerController :type Fort%20Player%20Controller%20Object%20Reference) (OldViewTarget :type Actor%20Object%20Reference) (NewViewTarget :type Actor%20Object%20Reference))
  (CallFunction|WaitforPlayerHotHandsTag NewViewTarget)
  (CallFunction|WaitforPlayerActionReload NewViewTarget))

(event Custom|WaitForPlayerHotHandsTag
  (params (PlayerActor :type Actor%20Object%20Reference))
  (bind _returnvalue (DynamicUIDirector|GetOwningLocalPlayerController))
  ;[1 1 1] Wait For Player Hot Hands Tag
  (Utilities|IsValid PlayerActor
    (:Is%20Valid
      (Utilities|IsValid _returnvalue
        (:Is%20Valid
          (Utilities|Casting|CastToFortPlayerPawnAthena (Pawn|GetControlledPawn _returnvalue))
          (Ability|Tasks|WaitGameplayTagAddToActor PlayerActor (Variables|Components|GetHotHandsTag_Event))
          (Ability|Tasks|WaitGameplayTagRemoveFromActor PlayerActor (Variables|Components|GetHotHandsTag_Event)))))))

(event Custom|WaitForPlayerActionReload
  (params (PlayerActor :type Actor%20Object%20Reference))
  (bind _returnvalue (DynamicUIDirector|GetOwningLocalPlayerController))
  ;[1 1 1] Wait For Player Action Reload
  (Utilities|IsValid PlayerActor
    (:Is%20Valid
      (Utilities|IsValid _returnvalue
        (:Is%20Valid
          (Utilities|Casting|CastToFortPlayerPawnAthena (Pawn|GetControlledPawn _returnvalue))
          (Ability|Tasks|WaitGameplayTagAddToActor PlayerActor (Variables|Components|GetActionReloadTag_Event))
          (Ability|Tasks|WaitGameplayTagRemoveFromActor PlayerActor (Variables|Components|GetActionReloadTag_Event)))))))

(fn ConstructionScript)
